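# Additive unlit glow material for Niagara mesh particles on the WeaveHarbor helmet cosmetic.
# Particle color RGB feeds emissive through MF_FX_EmissiveAdjustments, while particle alpha scales a view-angle mask.
# The UseHelmetGlow dynamic parameter blends that mask from a boosted Fresnel-style rim, the default, to a camera-facing falloff.
# A depth fade with adjustable FadeDistance softens the resulting opacity where it meets other geometry.
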Material /BRCosmetics/Characters/Player/Male/Medium/Bodies/M_MED_WeaveHarbor_Helmet/FX/M_WeaveHarbor_Helmet_Glow
graph {
 properties {
  bAutomaticallySetUsageInEditor=false
  bUsedWithNiagaraMeshParticles=true
  blendMode=BLEND_Additive
  pixelDepthOffsetMode=PDOM_Legacy
  previewMesh=&/SM_WeaveHarbor_FX_Helmet
  shadingModel=MSM_Unlit
  usedShadingModels=Unlit
 }
 nodes [
  n1 = ParticleColor()
  n2 = DotProduct(A=VertexNormalWS(), B=CameraVectorWS())
  n3 = OneMinus(None=@n2)
  n4 = Saturate(None=Multiply(A=Power(Base=@n3) {constExponent=4}) {constB=1.3})
  n5 = Power(Base=@n2) {constExponent=4}
  n6 = DynamicParameter() {defaultValue=<rgba 0 1 1 1>, paramNames [UseHelmetGlow, Param2, Param3, Param4]}
  DepthFade(
   Opacity=Multiply(A=@n1.A, B=LinearInterpolate(A=@n4, B=@n5, Alpha=@n6.UseHelmetGlow))
   FadeDistance=ScalarParameter() {defaultValue=10, parameterName=FadeDistance}
  )
  MaterialFunctionCall(Emissive=@n1.RGB) {materialFunction=&MF_FX_EmissiveAdjustments}
 ]
}
