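# Idle Niagara effect for the WeaveHarbor player helmet: red-to-violet lightning, side glows and lights attached to skeletal mesh sockets.
# Partitioned mesh, sprite and light emitters follow the sockets, while ribbon emitters bend between socket pairs with Hermite splines.
# System timelines blend arc reach, width and spawn rate, and a running flag enlarges emitted bolts.
# User.bisHeadHidden drives visibility tags and zeroes ribbon widths.
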NiagaraSystem /BRCosmetics/Characters/Player/Male/Medium/Bodies/M_MED_WeaveHarbor_Helmet/FX/NS_WeaveHarbor_Helmet_Idle
dependencies {
 usedDataInterfaces [
  &NiagaraDataInterfaceCurve
  &NiagaraDataInterfaceArrayFloat3
  &NiagaraDataInterfaceMeshRendererInfo
  &NiagaraDataInterfaceColorCurve
  &NiagaraDataInterfaceArrayInt32
  &NiagaraDataInterfaceSkeletalMesh
 ]
 usedDynamicInputs [
  &SetBoolByFloatComparison
  &VectorLength
  &Multiply_Vector
 ]
 usedModules [
  &self:SystemSpawnScript.SystemScriptSource.SystemScriptGraph.NiagaraNodeAssignment_0.SetVariables_#1
  &self:SystemSpawnScript.SystemScriptSource.SystemScriptGraph.NiagaraNodeAssignment_1.SetVariables_#2
  &SystemState
  &self:SystemSpawnScript.SystemScriptSource.SystemScriptGraph.NiagaraNodeAssignment_2.SetVariables_#3
  &self:SystemSpawnScript.SystemScriptSource.SystemScriptGraph.NiagaraNodeAssignment_3.SetVariables_#4
  &Timeline
 ]
}
emitters {
 CircleLightning {
  emitterUpdateScript [
   &EmitterState()
   &SpawnRate(SpawnRate=8, Use Spawn Probability=false)
  ]
  particleSpawnScript [
   &PartitionParticles()
   &InitializeParticle(
    Color=<rgba 1 0.0285 0 1>
    Lifetime Max=0.5
    Lifetime Min=0.3
    Lifetime Mode=::NewEnumerator1
    Mesh Scale Mode=::NewEnumerator3
    Mesh Uniform Scale=0.09
    UsePositionOffset=false
    Write Material Random=true
   )
   &CameraOffset(Camera Offset Amount=1)
   &InitialMeshOrientation(
    Mesh Orientation Mode=::NewEnumerator2
    Rotation=&RandomRangeVector(Minimum=<0 0 -1>, Maximum=<0 0 1>)
    Use Rotation Vector=true
   )
  ]
  particleUpdateScript [
   &ParticleState()
   &SkeletalMeshLocation(
    Bone / Socket Index=$Particles.PartitionParticles.Partition
    MaskSpawnGroup=false
    Mesh Sampling Type=::NewEnumerator5
    Skeletal Mesh=$System.Skeletal Mesh
    Socket Sampling Mode=::NewEnumerator5
   )
   &DynamicMaterialParameters(
    Index 0 Param 1=&FloatFromCurve(
     FloatCurve=DataInterface() {propertyValues {Curve {keys [
         {arriveTangent=4.1785, interpMode=RCIM_Cubic, leaveTangent=4.1785, tangentMode=RCTM_User, time=0, value=0}
         {arriveTangent=1.1179, interpMode=RCIM_Cubic, leaveTangent=1.1179, tangentMode=RCTM_User, time=0.2042, value=0.6228}
         {arriveTangent=0.0224, interpMode=RCIM_Cubic, leaveTangent=0.0224, tangentMode=RCTM_User, time=1, value=1}
        ]}}}
     Scale Curve=&RandomRangeFloat(Minimum=0, Maximum=1)
    )
    Index 0 Param 2=&RandomRangeFloat(Minimum=2, Maximum=3)
    Index 0 Param 3=&FloatFromCurve(
     FloatCurve=DataInterface() {propertyValues {Curve {keys [
         <at 0 1>
         {interpMode=RCIM_Cubic, leaveTangent=0.1085, tangentMode=RCTM_Break, time=0.1038, value=1}
         {interpMode=RCIM_Cubic, leaveTangent=-11.9441, tangentMode=RCTM_Break, time=0.2038, value=2.625}
         <at 1 0>
        ]}}}
     Scale Curve=&RandomRangeFloat(Minimum=15, Maximum=20)
    )
    Module.Index0Param0Write=true
    Module.Index0Param1Write=true
    Module.Index0Param2Write=true
    Module.Index0Param3Write=true
   )
   &ScaleColor(
    Scale Alpha=&FloatFromCurve(FloatCurve=DataInterface() {propertyValues {Curve {keys [
         {interpMode=RCIM_Cubic, leaveTangent=0.2357, tangentMode=RCTM_Break, time=0.3, value=1}
         {arriveTangent=-1.6327, tangentMode=RCTM_User, time=1, value=0}
        ]}}})
    Scale RGB=&VectorFromFloat(Value=&FloatFromCurve(FloatCurve=DataInterface() {propertyValues {Curve {keys [
          <at 0 1>
          {arriveTangent=1.25, leaveTangent=1.25, time=0.3, value=1}
          {interpMode=RCIM_Cubic, leaveTangent=-5.9679, tangentMode=RCTM_Break, time=0.4, value=1.5}
          <at 1 0.4>
         ]}}}))
    ScaleA=true
    ScaleRGB=true
   ) {module=ScaleColor001}
   /SetVariables(Particles.VisibilityTag=&MakeCustomIntFromBool(Bool=$User.bisHeadHidden, True INT=1, False INT=0)) {isSetParameters=true, module=SetVariables_#5}
  ]
  rendererClasses [&NiagaraMeshRendererProperties]
  renderers [
   {rendererClass=&NiagaraMeshRendererProperties, rendererIndex=0}
   {rendererClass=&NiagaraMeshRendererProperties, rendererIndex=1}
  ]
 }
 EmittingLightning {
  emitterUpdateScript [
   &EmitterState()
   &SpawnRate(SpawnRate=&Lerp_Float(A=4, B=8, Alpha=$System.Timeline.Time), Use Spawn Probability=false)
  ]
  particleSpawnScript [
   &InitializeParticle(
    Color Mode=::NewEnumerator0
    Lifetime Max=0.7
    Lifetime Min=0.45
    Lifetime Mode=::NewEnumerator1
    Mesh Renderer Array Visibility Mode=::NewEnumerator3
    Mesh Renderer Info=DataInterface() {propertyValues {MeshRenderer=&self:EmittingLightning_1.NiagaraMeshRendererProperties_1}}
    Mesh Scale=<0.5 0.2 1>
    Mesh Scale Mode=::NewEnumerator4
    UsePositionOffset=false
    Write Material Random=true
   )
   &PartitionParticles()
   &InitialMeshOrientation(Mesh Orientation Mode=::NewEnumerator2, Rotation=<-0.25 0 0.1>, Use Rotation Vector=true)
  ]
  particleUpdateScript [
   &ParticleState()
   &ScaleMeshSize(Scale Factor=&MakeVector(X=1, Y=1, Z=&MakeCustomFloatFromBool(Bool=$System.bIsRunning, True Float=1.6, False Float=1)))
   &DynamicMaterialParameters(
    Index 0 Param 1=&FloatFromCurve(FloatCurve=DataInterface() {propertyValues {Curve {keys [<at 0 1>,<at 1 0>]}}})
    Index 0 Param 2=&MakeCustomFloatFromBool(Bool=$System.bIsRunning, True Float=5, False Float=3)
    Index 0 Param 3=&FloatFromCurve(
     FloatCurve=DataInterface() {propertyValues {Curve {keys [<at 0 0.3>, {interpMode=RCIM_Cubic, leaveTangent=-5.5613, tangentMode=RCTM_Break, time=0.2, value=2}, <at 1 0>]}}}
     Scale Curve=50
    )
    Module.Index0Param0Write=true
    Module.Index0Param1Write=true
    Module.Index0Param2Write=true
    Module.Index0Param3Write=true
   )
   &SkeletalMeshLocation(
    Bone / Socket Index=$Particles.PartitionParticles.Partition
    MaskSpawnGroup=false
    Mesh Sampling Type=::NewEnumerator5
    Skeletal Mesh=$System.Skeletal Mesh
    Socket Sampling Mode=::NewEnumerator5
   )
   /SetVariables(Particles.VisibilityTag=&MakeCustomIntFromBool(Bool=$User.bisHeadHidden, True INT=1, False INT=0)) {isSetParameters=true, module=SetVariables_#5}
  ]
  rendererClasses [&NiagaraMeshRendererProperties]
  renderers [
   {rendererClass=&NiagaraMeshRendererProperties, rendererIndex=0}
   {rendererClass=&NiagaraMeshRendererProperties, rendererIndex=1}
  ]
 }
 IdleLightning_L {
  emitterSpawnScript [/SetVariables(Emitter.TextureScrollSpeed:Vector3f=<-1.5 0 7>) {isSetParameters=true, module=SetVariables_#6}]
  emitterUpdateScript [
   &EmitterState(Life Cycle Mode=::NewEnumerator1, Loop Behavior=::NewEnumerator1, Loop Duration Mode=::NewEnumerator1, UseLoopDelay=false)
   &SpawnBurst_Instantaneous(Spawn Count=6, Use Spawn Probability=false, UseLoopCountLimit=false)
  ]
  particleSpawnScript [
   &InitializeParticle(
    Lifetime=5
    Ribbon Width=&FloatFromCurve(
     FloatCurve=DataInterface() {propertyValues {Curve {keys [
         {arriveTangent=0.5303, interpMode=RCIM_Cubic, leaveTangent=0.5303, tangentMode=RCTM_User, time=0, value=0}
         {arriveTangent=1.781, interpMode=RCIM_Cubic, leaveTangent=-3.6314, tangentMode=RCTM_Break, time=0.5, value=0.85}
         {arriveTangent=-1.0106, interpMode=RCIM_Cubic, tangentMode=RCTM_Break, time=1, value=0}
        ]}}}
     CurveIndex=&ReturnNormalizedExecIndex()
     Scale Curve=&Lerp_Float(A=12, B=5, Alpha=$System.Timeline001.Time)
    )
    Ribbon Width Mode=::NewEnumerator1
    Sprite Size Mode=::NewEnumerator3
    Uniform Sprite Size=5
    UsePositionOffset=false
    Write Material Random=true
   )
   &PartitionParticles(Normalized Index In Partition Includes One=true, Particles Per Partition=16, Partition Mode:ENiagara_PartitionMode=::NewEnumerator1) {enabled=false}
   &DynamicMaterialParameters(
    Index 0 Param 1=0.8
    Index 0 Param 2=15
    Module.Index0Param0Write=true
    Module.Index0Param1Write=true
    Module.Index0Param2Write=true
    Module.Index0Param3Write=true
   )
   /SetVariables(
    Particles.Dy2:Vector4f=&MakeVector4(X=0.1, Y=0, Z=0, W=0)
    Particles.Linear Color=&ColorFromCurve(
     ColorCurve=DataInterface() {propertyValues {
       AlphaCurve {keys [<at 0.8537 1 RCIM_Cubic>]}
       BlueCurve {keys [<at 0.3078 0.0388 RCIM_Cubic>,<at 0.6636 0.014>,<at 1 1 RCIM_Cubic>]}
       GreenCurve {keys [<at 0.3078 0 RCIM_Cubic>,<at 0.6636 0>,<at 1 0.0249 RCIM_Cubic>]}
       RedCurve {keys [<at 0.3078 0.8542 RCIM_Cubic>,<at 0.6636 0.3073>,<at 1 0.2192 RCIM_Cubic>]}
      }}
     CurveIndex=&ReturnNormalizedExecIndex()
    )
   ) {isSetParameters=true, module=SetVariables_#7}
  ]
  particleUpdateScript [
   &ParticleState(Kill Particles When Lifetime Has Elapsed=false)
   /SetVariables(Particles.NewRibbonFacing:Vector3f=&TransformVector(Vector=<1 0 0>, Source Space=::Simulation, Destination Space=::Local, Transform Vector as Position=true)) {isSetParameters=true, module=SetVariables_#8}
   &SkeletalMeshLocation(
    MaskSpawnGroup=false
    Mesh Sampling Type=::NewEnumerator5
    Orientation Sampling=::NewEnumerator1
    Position Sampling=::NewEnumerator2
    Sampled Position Offset=&LerpPosition(A=<0 0 0>, B=<0 -0.5 0>, Alpha=$System.Timeline001.Time)
    Skeletal Mesh=$System.Skeletal Mesh
    Socket Sampling Mode=::NewEnumerator5
   )
   &SkeletalMeshLocation(
    Bone / Socket Index=6
    MaskSpawnGroup=false
    Mesh Sampling Type=::NewEnumerator5
    Orientation Sampling=::NewEnumerator1
    Position Sampling=::NewEnumerator2
    Sampled Position Offset=&Lerp_Vector(A=<0 0 0>, B=<-2 3 0>, Alpha=$System.Timeline001.Time)
    Skeletal Mesh=$System.Skeletal Mesh
    Socket Sampling Mode=::NewEnumerator5
   ) {module=SkeletalMeshLocation001}
   /Game/Effects/Niagara/Modules/Curves/FN_HermiteSpline(
    Position Blend:NiagaraFloat=&ReturnNormalizedExecIndex(Normalized Index Includes One=false)
    Position End:Vector3f=$Particles.SkeletalMeshLocation001.SampledPosition
    Position Start:Vector3f=$Particles.SkeletalMeshLocation.SampledPosition
    Write Particles Position=true
   )
   &CameraOffset(Camera Offset Amount=4)
   &ScaleRibbonWidth(Ribbon Width Scale=&MakeCustomFloatFromBool(Bool=$User.bisHeadHidden, True Float=0, False Float=1))
  ]
  rendererClasses [
   &NiagaraSpriteRendererProperties
   &NiagaraRibbonRendererProperties
   &NiagaraMeshRendererProperties
  ]
  renderers [
   {rendererClass=&NiagaraSpriteRendererProperties}
   {rendererClass=&NiagaraRibbonRendererProperties, rendererIndex=1}
   {rendererClass=&NiagaraMeshRendererProperties, rendererIndex=2}
   {rendererClass=&NiagaraRibbonRendererProperties, rendererIndex=3}
   {rendererClass=&NiagaraRibbonRendererProperties, rendererIndex=4}
  ]
 }
 IdleLightning_R {
  emitterSpawnScript [/SetVariables(Emitter.TextureScrollSpeed:Vector3f=<-1.5 0 7>) {isSetParameters=true, module=SetVariables_#6}]
  emitterUpdateScript [
   &EmitterState(Life Cycle Mode=::NewEnumerator1, Loop Behavior=::NewEnumerator1, Loop Duration Mode=::NewEnumerator1, UseLoopDelay=false)
   &SpawnBurst_Instantaneous(Spawn Count=6, Use Spawn Probability=false, UseLoopCountLimit=false)
  ]
  particleSpawnScript [
   &InitializeParticle(
    Lifetime=5
    Ribbon Width=&FloatFromCurve(
     FloatCurve=DataInterface() {propertyValues {Curve {keys [
         {arriveTangent=0.5303, interpMode=RCIM_Cubic, leaveTangent=0.5303, tangentMode=RCTM_User, time=0, value=0}
         {arriveTangent=1.781, interpMode=RCIM_Cubic, leaveTangent=-3.6314, tangentMode=RCTM_Break, time=0.5, value=0.875}
         {arriveTangent=-1.0106, interpMode=RCIM_Cubic, tangentMode=RCTM_Break, time=1, value=0}
        ]}}}
     CurveIndex=&ReturnNormalizedExecIndex()
     Scale Curve=&Lerp_Float(A=12, B=5, Alpha=$System.Timeline001.Time)
    )
    Ribbon Width Mode=::NewEnumerator1
    Sprite Size Mode=::NewEnumerator3
    Uniform Sprite Size=5
    UsePositionOffset=false
    Write Material Random=true
   )
   &PartitionParticles(Normalized Index In Partition Includes One=true, Particles Per Partition=16, Partition Mode:ENiagara_PartitionMode=::NewEnumerator1) {enabled=false}
   &DynamicMaterialParameters(
    Index 0 Param 1=0.8
    Index 0 Param 2=15
    Module.Index0Param0Write=true
    Module.Index0Param1Write=true
    Module.Index0Param2Write=true
    Module.Index0Param3Write=true
   )
   /SetVariables(
    Particles.Dy2:Vector4f=&MakeVector4(X=0.1, Y=0, Z=0, W=0)
    Particles.Linear Color=&ColorFromCurve(
     ColorCurve=DataInterface() {propertyValues {
       AlphaCurve {keys [<at 0.8537 1 RCIM_Cubic>]}
       BlueCurve {keys [<at 0.3078 0.0388 RCIM_Cubic>,<at 0.6636 0.014>,<at 1 1 RCIM_Cubic>]}
       GreenCurve {keys [<at 0.3078 0 RCIM_Cubic>,<at 0.6636 0>,<at 1 0.0249 RCIM_Cubic>]}
       RedCurve {keys [<at 0.3078 0.8542 RCIM_Cubic>,<at 0.6636 0.3073>,<at 1 0.2192 RCIM_Cubic>]}
      }}
     CurveIndex=&ReturnNormalizedExecIndex()
    )
   ) {isSetParameters=true, module=SetVariables_#7}
  ]
  particleUpdateScript [
   &ParticleState(Kill Particles When Lifetime Has Elapsed=false)
   /SetVariables(Particles.NewRibbonFacing:Vector3f=&TransformVector(Vector=<1 0 0>, Source Space=::Simulation, Destination Space=::Local, Transform Vector as Position=true)) {isSetParameters=true, module=SetVariables_#8}
   &SkeletalMeshLocation(
    Bone / Socket Index=1
    MaskSpawnGroup=false
    Mesh Sampling Type=::NewEnumerator5
    Orientation Sampling=::NewEnumerator1
    Position Sampling=::NewEnumerator2
    Sampled Position Offset=&LerpPosition(A=<0 0 0>, B=<0 -0.5 0>, Alpha=$System.Timeline001.Time)
    Skeletal Mesh=$System.Skeletal Mesh
    Socket Sampling Mode=::NewEnumerator5
   )
   &SkeletalMeshLocation(
    Bone / Socket Index=7
    MaskSpawnGroup=false
    Mesh Sampling Type=::NewEnumerator5
    Orientation Sampling=::NewEnumerator1
    Position Sampling=::NewEnumerator2
    Sampled Position Offset=&Lerp_Vector(A=<0 0 0>, B=<-2 3 0>, Alpha=$System.Timeline001.Time)
    Skeletal Mesh=$System.Skeletal Mesh
    Socket Sampling Mode=::NewEnumerator5
   ) {module=SkeletalMeshLocation001}
   &FN_HermiteSpline(
    Position Blend:NiagaraFloat=&ReturnNormalizedExecIndex(Normalized Index Includes One=false)
    Position End:Vector3f=$Particles.SkeletalMeshLocation001.SampledPosition
    Position Start:Vector3f=$Particles.SkeletalMeshLocation.SampledPosition
    Write Particles Position=true
   )
   &CameraOffset(Camera Offset Amount=4)
   &ScaleRibbonWidth(Ribbon Width Scale=&MakeCustomFloatFromBool(Bool=$User.bisHeadHidden, True Float=0, False Float=1))
  ]
  rendererClasses [
   &NiagaraSpriteRendererProperties
   &NiagaraRibbonRendererProperties
   &NiagaraMeshRendererProperties
  ]
  renderers [
   {rendererClass=&NiagaraSpriteRendererProperties}
   {rendererClass=&NiagaraRibbonRendererProperties, rendererIndex=1}
   {rendererClass=&NiagaraMeshRendererProperties, rendererIndex=2}
   {rendererClass=&NiagaraRibbonRendererProperties, rendererIndex=3}
   {rendererClass=&NiagaraRibbonRendererProperties, rendererIndex=4}
  ]
 }
 Light {
  emitterUpdateScript [
   &EmitterState()
   &SpawnRate(SpawnRate=6, Use Spawn Probability=false)
  ]
  particleSpawnScript [
   &PartitionParticles()
   &InitializeParticle(
    Color=<rgba 1 0.0081 0 1>
    Lifetime Max=0.7
    Lifetime Min=0.4
    Lifetime Mode=::NewEnumerator1
    Sprite Size Mode=::NewEnumerator3
    Uniform Sprite Size=5.8
    UsePositionOffset=false
    Write Material Random=true
   )
   &CameraOffset(Camera Offset Amount=1)
  ]
  particleUpdateScript [
   &ParticleState()
   &SkeletalMeshLocation(
    Bone / Socket Index=$Particles.PartitionParticles.Partition
    MaskSpawnGroup=false
    Mesh Sampling Type=::NewEnumerator5
    Sampled Position Offset=&SelectVectorFromArray(
     Direct Array Index=$Particles.PartitionParticles.Partition
     Vector Selection Array=DataInterface() {propertyValues {FloatData [<1 0 -3>, <1 0 3>]}}
    )
    Skeletal Mesh=$System.Skeletal Mesh
    Socket Sampling Mode=::NewEnumerator5
   )
   &FN_Light_Attributes(
    Color=$Particles.Initial.Color
    Create Color Attribute=true
    Create Exponent Attribute=true
    Create Radius Attribute=true
    Exponent=4
    Radius=15
    Scale Color=&FloatFromCurve(
     FloatCurve=DataInterface() {propertyValues {Curve {keys [
         {arriveTangent=9.0725, interpMode=RCIM_Cubic, leaveTangent=9.0725, tangentMode=RCTM_User, time=0, value=0}
         {arriveTangent=-2.5249, interpMode=RCIM_Cubic, leaveTangent=-2.5249, tangentMode=RCTM_User, time=0.1408, value=0.9847}
         {arriveTangent=-0.0405, interpMode=RCIM_Cubic, leaveTangent=-0.0405, tangentMode=RCTM_User, time=1, value=0}
        ]}}}
     Scale Curve=8
    )
   )
   /SetVariables(Particles.VisibilityTag=&MakeCustomIntFromBool(Bool=$User.bisHeadHidden, True INT=1, False INT=0)) {isSetParameters=true, module=SetVariables_#5}
  ]
  rendererClasses [&NiagaraLightRendererProperties]
  renderers [{rendererClass=&NiagaraLightRendererProperties, rendererIndex=0}]
 }
 LightningBG {
  emitterUpdateScript [
   &EmitterState(Life Cycle Mode=::NewEnumerator1, Loop Behavior=::NewEnumerator1, Loop Duration=1, UseLoopDelay=false)
   &SpawnBurst_Instantaneous(Spawn Count=2, Use Spawn Probability=false, UseLoopCountLimit=false)
  ]
  particleSpawnScript [
   &InitializeParticle(
    Color=&FNTimeOfDayColor(Night Color=<rgba 1 1 1 0.25>, Morning Color=<rgba 1 1 1 0.4>, Day Color=<rgba 1 1 1 0.45>, Evening Color=<rgba 1 1 1 0.5>)
    Lifetime=5
    Mesh Scale=<1 1.3 1>
    Mesh Scale Mode=::NewEnumerator4
    UsePositionOffset=false
    Write Material Random=true
   )
   &PartitionParticles()
   &InitialMeshOrientation(Mesh Orientation Mode=::NewEnumerator2, Orientation Vector=<0 0 1>, Rotation=<0.1419 0 0>, Use Rotation Vector=true)
  ]
  particleUpdateScript [
   &ParticleState(Kill Particles When Lifetime Has Elapsed=false)
   &SkeletalMeshLocation(
    Bone / Socket Index=$Particles.PartitionParticles.Partition
    MaskSpawnGroup=false
    Mesh Sampling Type=::NewEnumerator5
    Skeletal Mesh=$System.Skeletal Mesh
    Socket Sampling Mode=::NewEnumerator5
   )
   &ScaleMeshSize(Scale Factor=&MakeVector(X=1, Y=&Lerp_Float(A=1, B=0.8, Alpha=$System.Timeline001.Time), Z=1))
   /SetVariables(Particles.VisibilityTag=&MakeCustomIntFromBool(Bool=$User.bisHeadHidden, True INT=1, False INT=0)) {isSetParameters=true, module=SetVariables_#5}
  ]
  rendererClasses [&NiagaraMeshRendererProperties]
  renderers [{rendererClass=&NiagaraMeshRendererProperties, rendererIndex=0}]
 }
 ReactiveLightning {
  emitterUpdateScript [
   &EmitterState(Life Cycle Mode=::NewEnumerator1, Loop Behavior=::NewEnumerator1, Loop Duration Mode=::NewEnumerator1, UseLoopDelay=false)
   &SpawnBurst_Instantaneous(Spawn Count=32, Use Spawn Probability=false, UseLoopCountLimit=false)
  ]
  particleSpawnScript [
   &InitializeParticle(
    Color=&ColorFromCurve(
     ColorCurve=DataInterface() {propertyValues {
       AlphaCurve {keys [<at 0.2685 0>,<at 0.4505 2>,<at 0.5961 2>,<at 1 0>]}
       BlueCurve {keys [<at 0.1156 0.0482 RCIM_Cubic>,<at 0.5178 0.0216>,<at 0.82 1 RCIM_Cubic>]}
       GreenCurve {keys [<at 0.1156 0 RCIM_Cubic>,<at 0.5178 0>,<at 0.82 0.0442 RCIM_Cubic>]}
       RedCurve {keys [<at 0.1156 1 RCIM_Cubic>,<at 0.5178 0.4479>,<at 0.82 0.2599 RCIM_Cubic>]}
      }}
     CurveIndex=&ReturnNormalizedExecIndex()
    )
    Lifetime=5
    Ribbon Width=&FloatFromCurve(
     FloatCurve=DataInterface() {propertyValues {Curve {keys [
         {arriveTangent=3.1474, interpMode=RCIM_Cubic, leaveTangent=3.1474, tangentMode=RCTM_User, time=0, value=-0.05}
         {arriveTangent=0.1963, interpMode=RCIM_Cubic, leaveTangent=-0.068, tangentMode=RCTM_Break, time=0.4, value=1.075}
         {arriveTangent=-1.8225, interpMode=RCIM_Cubic, tangentMode=RCTM_Break, time=1, value=0}
        ]}}}
     CurveIndex=&ReturnNormalizedExecIndex()
     Scale Curve=7
    )
    Ribbon Width Mode=::NewEnumerator1
    UsePositionOffset=false
    Write Material Random=true
   )
   &PartitionParticles()
   &DynamicMaterialParameters(
    Index 0 Param 1=0.4
    Index 0 Param 4=1
    Module.Index0Param0Write=true
    Module.Index0Param1Write=true
    Module.Index0Param2Write=true
    Module.Index0Param3Write=true
   )
   /SetVariables(
    Particles.Dy2:Vector4f=<0.9 50 0 0>
    Particles.Dy3:Vector4f=<0.61 0.41 0 0>
    Particles.Linear Color=&ColorFromCurve(
     ColorCurve=DataInterface() {propertyValues {
       AlphaCurve {keys [<at 0 0>,<at 0.1491 0.7>,<at 0.4264 0.4>,<at 1 0>]}
       BlueCurve {keys [<at 0.1156 0.0482 RCIM_Cubic>,<at 0.5178 0.0216>,<at 0.82 1 RCIM_Cubic>]}
       GreenCurve {keys [<at 0.1156 0 RCIM_Cubic>,<at 0.5178 0>,<at 0.82 0.0442 RCIM_Cubic>]}
       RedCurve {keys [<at 0.1156 1 RCIM_Cubic>,<at 0.5178 0.4479>,<at 0.82 0.2599 RCIM_Cubic>]}
      }}
     CurveIndex=&ReturnNormalizedExecIndex(Normalized Index Includes One=false)
    )
   ) {isSetParameters=true}
  ]
  particleUpdateScript [
   &ParticleState(Kill Particles When Lifetime Has Elapsed=false, Loop Particles Lifetime=true)
   /SetVariables(Particles.NewRibbonFacing:Vector3f=&TransformVector(Vector=<1 0 0>, Source Space=::Simulation, Destination Space=::Local, Transform Vector as Position=true)) {enabled=false, isSetParameters=true, module=SetVariables_#8}
   &SkeletalMeshLocation(
    Bone / Socket Index=$Particles.PartitionParticles.Partition
    MaskSpawnGroup=false
    Mesh Sampling Type=::NewEnumerator5
    Orientation Sampling=::NewEnumerator1
    Position Sampling=::NewEnumerator1
    Sampled Position Offset=&SelectVectorFromArray(
     Direct Array Index=$Particles.PartitionParticles.Partition
     Vector Selection Array=DataInterface() {propertyValues {FloatData [<0 0 0>, <0 0 0>]}}
    )
    Skeletal Mesh=$System.Skeletal Mesh
    Socket Sampling Mode=::NewEnumerator5
   )
   &SkeletalMeshLocation(
    Bone / Socket Index=&SelectIntFromArray(Direct Array Index=$Particles.PartitionParticles.Partition, Int Selection Array=DataInterface() {propertyValues {IntData [6,7]}})
    MaskSpawnGroup=false
    Mesh Sampling Type=::NewEnumerator5
    Orientation Sampling=::NewEnumerator1
    Position Sampling=::NewEnumerator1
    Sampled Position Offset=<5 -6 -1>
    Skeletal Mesh=$System.Skeletal Mesh
    Socket Sampling Mode=::NewEnumerator5
    Source Movement Scale:Vector3f=<0.4 0.4 0.4>
    Write To Particle Position=false
   ) {module=SkeletalMeshLocation001}
   &FN_HermiteSpline(
    Position Blend:NiagaraFloat=&ReturnNormalizedExecIndex(Normalized Index Includes One=false)
    Position End:Vector3f=&Lerp_Vector(A=$Particles.SkeletalMeshLocation.SampledPosition, B=$Particles.SkeletalMeshLocation001.SampledPosition, Alpha=$System.Timeline.Time)
    Position Start:Vector3f=$Particles.SkeletalMeshLocation.SampledPosition
    Write Particles Position=true
   )
   &CameraOffset(Camera Offset Amount=4)
   &ScaleRibbonWidth(Ribbon Width Scale=&MakeCustomFloatFromBool(Bool=$User.bisHeadHidden, True Float=0, False Float=1))
  ]
  rendererClasses [&NiagaraRibbonRendererProperties]
  renderers [
   {rendererClass=&NiagaraRibbonRendererProperties}
   {rendererClass=&NiagaraRibbonRendererProperties, rendererIndex=1}
   {rendererClass=&NiagaraRibbonRendererProperties, rendererIndex=2}
  ]
 }
 SideGlowOnHelmet {
  emitterUpdateScript [
   &EmitterState(Life Cycle Mode=::NewEnumerator1, Loop Behavior=::NewEnumerator1, Loop Duration=1, UseLoopDelay=false)
   &SpawnBurst_Instantaneous(Spawn Count=2, Use Spawn Probability=false, UseLoopCountLimit=false)
  ]
  particleSpawnScript [
   &PartitionParticles()
   &InitializeParticle(
    Color=<rgba 0.7865 0.0064 0 1>
    Lifetime=5
    Sprite Size Mode=::NewEnumerator3
    Uniform Sprite Size=7
    UsePositionOffset=false
    Write Material Random=true
   )
   &CameraOffset(Camera Offset Amount=1)
  ]
  particleUpdateScript [
   &ParticleState(Kill Particles When Lifetime Has Elapsed=false)
   &SkeletalMeshLocation(
    Bone / Socket Index=$Particles.PartitionParticles.Partition
    MaskSpawnGroup=false
    Mesh Sampling Type=::NewEnumerator5
    Skeletal Mesh=$System.Skeletal Mesh
    Socket Sampling Mode=::NewEnumerator5
   )
   /SetVariables(
    Particles.SpriteAlignment=$Particles.SkeletalMeshLocation.SampledPosition
    Particles.SpriteFacing=&RotateVectorByQuaternion(
     Vector To Rotate=&SelectVectorFromArray(
      Direct Array Index=$Particles.PartitionParticles.Partition
      Vector Selection Array=DataInterface() {propertyValues {FloatData [<0 0 -1>, <0 0 1>]}}
     )
     Quaternion=$Particles.SkeletalMeshLocation.SampledRotation
    )
   ) {isSetParameters=true}
   &DynamicMaterialParameters(
    Index 0 Param 1=1
    Index 0 Param 2=0.8668
    Module.Index0Param0Write=true
    Module.Index0Param1Write=true
    Module.Index0Param2Write=true
    Module.Index0Param3Write=true
   )
   /SetVariables(
    Particles.Linear Color=<rgba 3 2.2493 2.2792 1>
    Particles.RingSize:Vector2f=&Vector2DFromFloat(Value=4.5)
   ) {isSetParameters=true}
   /SetVariables(Particles.VisibilityTag=&MakeCustomIntFromBool(Bool=$User.bisHeadHidden, True INT=1, False INT=0)) {isSetParameters=true, module=SetVariables_#5}
  ]
  rendererClasses [&NiagaraSpriteRendererProperties]
  renderers [
   {rendererClass=&NiagaraSpriteRendererProperties}
   {rendererClass=&NiagaraSpriteRendererProperties, rendererIndex=1}
  ]
 }
}
summary {
 systemName=NS_WeaveHarbor_Helmet_Idle
 emitters [
  {emitterName=CircleLightning, rendererClasses [&NiagaraMeshRendererProperties]}
  {emitterName=SideGlowOnHelmet, rendererClasses [&NiagaraSpriteRendererProperties]}
  {emitterName=Light, rendererClasses [&NiagaraLightRendererProperties]}
  {emitterName=LightningBG, rendererClasses [&NiagaraMeshRendererProperties]}
  {emitterName=EmittingLightning, rendererClasses [&NiagaraMeshRendererProperties]}
  {emitterName=ReactiveLightning, rendererClasses [&NiagaraRibbonRendererProperties]}
  {
   emitterName=IdleLightning_R
   rendererClasses [
    &NiagaraSpriteRendererProperties
    &NiagaraRibbonRendererProperties
    &NiagaraMeshRendererProperties
   ]
  }
  {
   emitterName=IdleLightning_L
   rendererClasses [
    &NiagaraSpriteRendererProperties
    &NiagaraRibbonRendererProperties
    &NiagaraMeshRendererProperties
   ]
  }
 ]
 userVariables [
  User.bIsAiming {type=NiagaraBool}
  User.bisHeadHidden {type=NiagaraBool}
  User.RibbonIDOffset {type=NiagaraInt32}
 ]
}
